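# Additive, unlit, two-sided mesh-particle material for the ground wave effect in the Ninja's Kunai Storm ability.
# Particle color feeds a shared emissive adjustment function.
# A separate mask multiplies particle alpha, vertex red and a noise texture sampled by U alone, shifted by the doubled dynamic Time parameter.
# A sine of squared, halved depth fade also multiplies in, forming a soft band near intersecting geometry.
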Material /Game/Abilities/Player/Ninja/Actives/KunaiStorm/FX/M_GroundWave
graph {
 properties {
  bUsedWithMeshParticles=true
  blendMode=BLEND_Additive
  pixelDepthOffsetMode=PDOM_Legacy
  previewMesh=&/S_GroundWave
  shadingModel=MSM_Unlit
  twoSided=true
  usedShadingModels=Unlit
 }
 nodes [
  n1 = DepthFade(FadeDistance=ScalarParameter() {defaultValue=25, parameterName=DepthFade})
  ComponentMask(None=@n1) {r=true}
  n2 = ParticleColor()
  MaterialFunctionCall(Emissive=@n2.RGB) {materialFunction=&MF_FX_EmissiveAdjustments}
  n3 = VertexColor()
  n4 = Multiply(A=Multiply(A=@n1, B=@n1)) {constB=0.5}
  n5 = ComponentMask(None=TextureCoordinate() {uTiling=0.5, vTiling=0.5}) {r=true}
  n6 = DynamicParameter() {paramNames [Time,Random,Param3,Param4]}
  n7 = Add(A=@n6.Time, B=@n6.Time)
  Multiply(
   A=Multiply(A=Multiply(A=@n2.A, B=@n3.R), B=Sine(None=@n4))
   B=TextureSample(UVs=Add(A=@n5, B=@n7)) {texture=/Game/Effects/Fort_Effects/Textures/Misc/T_Shot0030_NoiseCutout}.R
  )
 ]
}
